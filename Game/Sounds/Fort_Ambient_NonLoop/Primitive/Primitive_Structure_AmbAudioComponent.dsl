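; Actor component that plays intermittent one-shot ambient sounds around its owning actor, likely a primitive structure.
; On BeginPlay, where cosmetic events can execute, it starts a timer with a random delay between minimum and maximum values.
; Each firing reschedules the timer and plays a random sound from SoundToPlay at a random point within the owner's bounds.
; EndPlay clears the timer on the same cosmetic-only condition.
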(event EventEndPlay
  (params (EndPlayReason :type EEndPlayReason%20Enum :default "Destroyed"))
  (Networking|CanExecuteCosmeticEvents
    (:True
      (Utilities|Time|ClearandInvalidateTimerbyHandle (Variables|Default|GetTimerHandle)))))

(event EventBeginPlay
  (Networking|CanExecuteCosmeticEvents
    (:True
      (bind _returnvalue (Utilities|Time|SetTimerbyEvent (EventDispatchers|CreateEvent :!event Custom|Play%20Ambient%20Audio) (Math|Random|RandomFloatinRange (Variables|Default|GetRandomTimeMin) (Variables|Default|GetRandomTimeMax))))
      (Variables|Default|SetTimerHandle _returnvalue))))

(event Custom|Play%20Ambient%20Audio
  (bind (_origin _boxextent) (Collision|GetActorBounds (Components|GetOwner) false false))
  (Utilities|FlowControl|Sequence
    (:then_0
      (bind _returnvalue (Utilities|Time|SetTimerbyEvent (EventDispatchers|CreateEvent :!event Custom|Play%20Ambient%20Audio) (Math|Random|RandomFloatinRange (Variables|Default|GetRandomTimeMin) (Variables|Default|GetRandomTimeMax))))
      (Variables|Default|SetTimerHandle _returnvalue))
    (:then_1
      (Audio|PlaySoundatLocation (Utilities|Array|RandomArrayItem (Variables|Default|GetSoundToPlay)) (Math|Random|RandomPointinBoundingBox _origin _boxextent)))))
