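# Sound cue for a primitive windmill ambient sound, assigned to the Athena outdoor ambient sound class.
# A single looping wave player feeds Primitive_Windmill_01 straight to the output at reduced volume, despite the cue sitting in a NonLoop folder.
# An attenuation override applies logarithmic distance falloff and a low-pass filter over distance.
# Listener focus with interpolation lowers the volume when the sound is out of focus.
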SoundCue /Game/Sounds/Fort_Ambient_NonLoop/Primitive/Primitive_Windmill_Cue
root=WavePlayer_0
nodes [{id=WavePlayer_0, properties {bLooping=true, soundWaveAssetPtr=&/Primitive_Windmill_01}}]
properties {
 bOverrideAttenuation=true
 maxDistance=4166.6143
 soundClassObject=&Fort_Ambient_Outdoors_Athena
 volumeMultiplier=0.226
 attenuationOverrides {
  bAttenuateWithLPF=true
  bEnableFocusInterpolation=true
  bEnableListenerFocus=true
  distanceAlgorithm=Logarithmic
  falloffDistance=3393.2688
  lPFFrequencyAtMin=1853.1152
  lPFRadiusMax=4000
  lPFRadiusMin=2000
  nonFocusVolumeAttenuation=0.5
  attenuationShapeExtents {x=773.3454}
 }
}
